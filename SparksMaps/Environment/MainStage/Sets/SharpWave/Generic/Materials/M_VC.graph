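# M_VC is a minimal Default Lit material in the generic materials of the SharpWave main-stage environment set.
# Its graph holds a single Vertex Color node, with no textures, parameters or math, and no link to a material output is shown.
# Automatic setting of usage flags in the editor is turned off.
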Material /SparksMaps/Environment/MainStage/Sets/SharpWave/Generic/Materials/M_VC
graph {properties {bAutomaticallySetUsageInEditor=false, usedShadingModels=Default Lit}, nodes [VertexColor()]}
